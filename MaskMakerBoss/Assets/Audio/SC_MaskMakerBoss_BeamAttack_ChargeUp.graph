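# Sound cue for the charge-up of the MaskMaker boss's beam attack, using boss-fight SFX class and skill attenuation.
# A mixer layers a random charge-up A variant, a fixed, quieter charge-up C wave, a random beam charge, and a randomly chosen synth whoosh delayed by a fixed interval.
# The mix passes through a root modulator, and overridden concurrency caps simultaneous instances by stopping the oldest.
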SoundCue /MaskMakerBoss/Assets/Audio/SC_MaskMakerBoss_BeamAttack_ChargeUp
root=Modulator_0
nodes [
 {id=Delay_0, children [Random_4], properties {delayMax=0.8, delayMin=0.8}}
 {id=Mixer_0, children [Random_1,WavePlayer_0,Delay_0,Random_0], properties {inputVolume [1.1,0.5,1,1]}}
 {id=Modulator_0, children [Mixer_0]}
 {id=Random_0, children [WavePlayer_1,WavePlayer_2], properties {weights [1,1]}}
 {id=Random_1, children [WavePlayer_3,WavePlayer_4,WavePlayer_5], properties {weights [1,1,1]}}
 {
  id=Random_4
  children [WavePlayer_20, WavePlayer_21, WavePlayer_22, WavePlayer_23, WavePlayer_24, WavePlayer_25]
  properties {weights [1,1,1,1,1,1]}
 }
 {id=WavePlayer_0, properties {soundWaveAssetPtr=&/MaskMakerBoss_BeamAttack_ChargeUp_C_02}}
 {id=WavePlayer_1, properties {soundWaveAssetPtr=&/MaskMakerBoss_BeamCharge_02}}
 {id=WavePlayer_2, properties {soundWaveAssetPtr=&/MaskMakerBoss_BeamCharge_01}}
 {id=WavePlayer_20, properties {soundWaveAssetPtr=&/MaskMakerBoss_SynthWhooshA_01}}
 {id=WavePlayer_21, properties {soundWaveAssetPtr=&/MaskMakerBoss_SynthWhooshA_02}}
 {id=WavePlayer_22, properties {soundWaveAssetPtr=&/MaskMakerBoss_SynthWhooshA_03}}
 {id=WavePlayer_23, properties {soundWaveAssetPtr=&/MaskMakerBoss_SynthWhooshA_04}}
 {id=WavePlayer_24, properties {soundWaveAssetPtr=&/MaskMakerBoss_SynthWhooshA_05}}
 {id=WavePlayer_25, properties {soundWaveAssetPtr=&/MaskMakerBoss_SynthWhooshA_06}}
 {id=WavePlayer_3, properties {soundWaveAssetPtr=&/MaskMakerBoss_BeamAttack_ChargeUp_A_01}}
 {id=WavePlayer_4, properties {soundWaveAssetPtr=&/MaskMakerBoss_BeamAttack_ChargeUp_A_02}}
 {id=WavePlayer_5, properties {soundWaveAssetPtr=&/MaskMakerBoss_BeamAttack_ChargeUp_A_03}}
]
properties {
 attenuationSettings=&/ATT_MaskMakerBoss_Skills
 bDebug=true
 bOverrideConcurrency=true
 maxDistance=13500
 soundClassObject=&Fort_BossFight_SFX
 volumeMultiplier=0.95
 concurrencyOverrides {maxCount=3, resolutionRule=StopOldest}
}
