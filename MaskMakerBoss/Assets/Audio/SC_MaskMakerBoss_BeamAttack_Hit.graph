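# Sound cue for the hit sound of the Mask Maker boss's beam attack.
# A mixer layers two equally weighted random pickers at equal volume, each choosing one of six wave variants from the HitA and HitC sets.
# It uses the boss skills attenuation and the boss fight SFX class, with a concurrency override that caps simultaneous instances and debug enabled.
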SoundCue /MaskMakerBoss/Assets/Audio/SC_MaskMakerBoss_BeamAttack_Hit
root=Mixer_1
nodes [
 {id=Mixer_1, children [Random_3,Random_2], properties {inputVolume [1,1]}}
 {
  id=Random_2
  children [WavePlayer_22, WavePlayer_23, WavePlayer_24, WavePlayer_25, WavePlayer_26, WavePlayer_27]
  properties {weights [1,1,1,1,1,1]}
 }
 {
  id=Random_3
  children [WavePlayer_9, WavePlayer_10, WavePlayer_11, WavePlayer_12, WavePlayer_20, WavePlayer_21]
  properties {weights [1,1,1,1,1,1]}
 }
 {id=WavePlayer_10, properties {soundWaveAssetPtr=&/MaskMakerBoss_BeamAttack_HitA_02}}
 {id=WavePlayer_11, properties {soundWaveAssetPtr=&/MaskMakerBoss_BeamAttack_HitA_03}}
 {id=WavePlayer_12, properties {soundWaveAssetPtr=&/MaskMakerBoss_BeamAttack_HitA_04}}
 {id=WavePlayer_20, properties {soundWaveAssetPtr=&/MaskMakerBoss_BeamAttack_HitA_05}}
 {id=WavePlayer_21, properties {soundWaveAssetPtr=&/MaskMakerBoss_BeamAttack_HitA_06}}
 {id=WavePlayer_22, properties {soundWaveAssetPtr=&/MaskMakerBoss_BeamAttack_HitC_01}}
 {id=WavePlayer_23, properties {soundWaveAssetPtr=&/MaskMakerBoss_BeamAttack_HitC_02}}
 {id=WavePlayer_24, properties {soundWaveAssetPtr=&/MaskMakerBoss_BeamAttack_HitC_03}}
 {id=WavePlayer_25, properties {soundWaveAssetPtr=&/MaskMakerBoss_BeamAttack_HitC_04}}
 {id=WavePlayer_26, properties {soundWaveAssetPtr=&/MaskMakerBoss_BeamAttack_HitC_05}}
 {id=WavePlayer_27, properties {soundWaveAssetPtr=&/MaskMakerBoss_BeamAttack_HitC_06}}
 {id=WavePlayer_9, properties {soundWaveAssetPtr=&/MaskMakerBoss_BeamAttack_HitA_01}}
]
properties {
 attenuationSettings=&/ATT_MaskMakerBoss_Skills
 bDebug=true
 bOverrideConcurrency=true
 maxDistance=13500
 soundClassObject=&Fort_BossFight_SFX
 volumeMultiplier=0.7
 concurrencyOverrides {maxCount=2}
}
