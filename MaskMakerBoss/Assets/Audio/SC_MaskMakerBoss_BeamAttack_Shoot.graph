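# Sound cue for the MaskMaker boss's beam attack shot, routed to the boss fight SFX class with boss skill attenuation.
# A mixer layers five random pools: two shoot sets, synth whooshes, sparks and beam shots, with the first shoot and beam shot layers boosted.
# Modulators raise pitch randomly on both shoot sets and the whooshes.
# Overridden concurrency allows two instances, stopping the oldest, and the debug flag is on.
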SoundCue /MaskMakerBoss/Assets/Audio/SC_MaskMakerBoss_BeamAttack_Shoot
root=Mixer_1
nodes [
 {id=Mixer_1, children [Modulator_2,Modulator_1,Modulator_0,Random_2,Random_3], properties {inputVolume [1.2,1,1,1,1.2]}}
 {id=Modulator_0, children [Random_6], properties {pitchMax=1.2, pitchMin=1}}
 {id=Modulator_1, children [Random_0], properties {pitchMax=1.2, pitchMin=1}}
 {id=Modulator_2, children [Random_4], properties {pitchMax=1.2, pitchMin=1}}
 {id=Random_0, children [WavePlayer_15, WavePlayer_16, WavePlayer_17, WavePlayer_18, WavePlayer_19], properties {weights [1,1,1,1,1]}}
 {id=Random_2, children [WavePlayer_9, WavePlayer_10, WavePlayer_11], properties {weights [1,1,1]}}
 {id=Random_3, children [WavePlayer_12, WavePlayer_20, WavePlayer_21], properties {weights [1,1,1]}}
 {id=Random_4, children [WavePlayer_0, WavePlayer_1, WavePlayer_2, WavePlayer_13, WavePlayer_14], properties {weights [1,1,1,1,1]}}
 {
  id=Random_6
  children [WavePlayer_32, WavePlayer_33, WavePlayer_34, WavePlayer_35, WavePlayer_36, WavePlayer_37]
  properties {weights [1,1,1,1,1,1]}
 }
 {id=WavePlayer_0, properties {soundWaveAssetPtr=&/MaskMakerBoss_BeamAttack_Shoot_01}}
 {id=WavePlayer_1, properties {soundWaveAssetPtr=&/MaskMakerBoss_BeamAttack_Shoot_02}}
 {id=WavePlayer_10, properties {soundWaveAssetPtr=&/MaskMaker_Sparks_02}}
 {id=WavePlayer_11, properties {soundWaveAssetPtr=&/MaskMaker_Sparks_03}}
 {id=WavePlayer_12, properties {soundWaveAssetPtr=&/MaskMakerBoss_BeamShoot_01}}
 {id=WavePlayer_13, properties {soundWaveAssetPtr=&/MaskMakerBoss_BeamAttack_Shoot_04}}
 {id=WavePlayer_14, properties {soundWaveAssetPtr=&/MaskMakerBoss_BeamAttack_Shoot_05}}
 {id=WavePlayer_15, properties {soundWaveAssetPtr=&/MaskMakerBoss_BeamAttack_ShootB_01}}
 {id=WavePlayer_16, properties {soundWaveAssetPtr=&/MaskMakerBoss_BeamAttack_ShootB_02}}
 {id=WavePlayer_17, properties {soundWaveAssetPtr=&/MaskMakerBoss_BeamAttack_ShootB_03}}
 {id=WavePlayer_18, properties {soundWaveAssetPtr=&/MaskMakerBoss_BeamAttack_ShootB_04}}
 {id=WavePlayer_19, properties {soundWaveAssetPtr=&/MaskMakerBoss_BeamAttack_ShootB_05}}
 {id=WavePlayer_2, properties {soundWaveAssetPtr=&/MaskMakerBoss_BeamAttack_Shoot_03}}
 {id=WavePlayer_20, properties {soundWaveAssetPtr=&/MaskMakerBoss_BeamShoot_02}}
 {id=WavePlayer_21, properties {soundWaveAssetPtr=&/MaskMakerBoss_BeamShoot_03}}
 {id=WavePlayer_32, properties {soundWaveAssetPtr=&/MaskMakerBoss_SynthWhooshB_06}}
 {id=WavePlayer_33, properties {soundWaveAssetPtr=&/MaskMakerBoss_SynthWhooshB_05}}
 {id=WavePlayer_34, properties {soundWaveAssetPtr=&/MaskMakerBoss_SynthWhooshB_04}}
 {id=WavePlayer_35, properties {soundWaveAssetPtr=&/MaskMakerBoss_SynthWhooshB_03}}
 {id=WavePlayer_36, properties {soundWaveAssetPtr=&/MaskMakerBoss_SynthWhooshB_02}}
 {id=WavePlayer_37, properties {soundWaveAssetPtr=&/MaskMakerBoss_SynthWhooshB_01}}
 {id=WavePlayer_9, properties {soundWaveAssetPtr=&/MaskMaker_Sparks_01}}
]
properties {
 attenuationSettings=&/ATT_MaskMakerBoss_Skills
 bDebug=true
 bOverrideConcurrency=true
 maxDistance=13500
 soundClassObject=&Fort_BossFight_SFX
 volumeMultiplier=0.8
 concurrencyOverrides {maxCount=2, resolutionRule=StopOldest}
}
