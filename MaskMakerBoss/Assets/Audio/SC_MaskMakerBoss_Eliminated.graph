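# Sound cue for the Mask Maker boss's elimination, layering seven weighted inputs through one mixer.
# Synth whoosh, dematerialize, beam-hit and eliminated layers start immediately; burning, Deimos spawner explosion and a second whoosh follow on fixed delays.
# Random nodes vary all but the eliminated wave, with one explosion option adding extra delay.
# It uses boss skill attenuation, boss-fight SFX class, disabled virtualization and a tight concurrency override with volume scaling.
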SoundCue /MaskMakerBoss/Assets/Audio/SC_MaskMakerBoss_Eliminated
root=Mixer_0
nodes [
 {id=Delay_15, children [WavePlayer_20], properties {delayMax=0.25, delayMin=0.25}}
 {id=Delay_2, children [Random_6], properties {delayMax=0.75, delayMin=0.75}}
 {id=Delay_5, children [Random_10], properties {delayMax=0.75, delayMin=0.75}}
 {id=Delay_7, children [Random_9], properties {delayMax=1, delayMin=1}}
 {
  id=Mixer_0
  children [Random_8, Random_1, Delay_5, Delay_2, Delay_7, Random_11, WavePlayer_11]
  properties {inputVolume [1,0.75,1,0.75,0.75,1,0.6]}
 }
 {id=Random_1, children [WavePlayer_0,WavePlayer_1,WavePlayer_2,WavePlayer_3], properties {weights [1,1,1,1]}}
 {id=Random_10, children [WavePlayer_21, Delay_15], properties {weights [1,1]}}
 {
  id=Random_11
  children [WavePlayer_51, WavePlayer_4, WavePlayer_5, WavePlayer_6, WavePlayer_7, WavePlayer_8]
  properties {weights [1,1,1,1,1,1]}
 }
 {id=Random_6, children [WavePlayer_27, WavePlayer_28], properties {weights [1,1]}}
 {
  id=Random_8
  children [WavePlayer_36, WavePlayer_37, WavePlayer_38, WavePlayer_39, WavePlayer_40, WavePlayer_41]
  properties {weights [1,1,1,1,1,1]}
 }
 {
  id=Random_9
  children [WavePlayer_42, WavePlayer_43, WavePlayer_44, WavePlayer_45, WavePlayer_46, WavePlayer_47]
  properties {weights [1,1,1,1,1,1]}
 }
 {id=WavePlayer_0, properties {soundWaveAssetPtr=&/NPC_Eliminated_DeMaterialize_01}}
 {id=WavePlayer_1, properties {soundWaveAssetPtr=&/NPC_Eliminated_DeMaterialize_02}}
 {id=WavePlayer_11, properties {soundWaveAssetPtr=&/MaskMakerBoss_Eliminated_01}}
 {id=WavePlayer_2, properties {soundWaveAssetPtr=&/NPC_Eliminated_DeMaterialize_03}}
 {id=WavePlayer_20, properties {soundWaveAssetPtr=&/DeimosSpawner_Explode}}
 {id=WavePlayer_21, properties {soundWaveAssetPtr=&/DeimosSpawner_Stage2Explosion}}
 {id=WavePlayer_27, properties {soundWaveAssetPtr=&/MaskMakerBoss_Burning_01}}
 {id=WavePlayer_28, properties {soundWaveAssetPtr=&/MaskMakerBoss_Burning_02}}
 {id=WavePlayer_3, properties {soundWaveAssetPtr=&/NPC_Eliminated_DeMaterialize_04}}
 {id=WavePlayer_36, properties {soundWaveAssetPtr=&/MaskMakerBoss_SynthWhooshC_06}}
 {id=WavePlayer_37, properties {soundWaveAssetPtr=&/MaskMakerBoss_SynthWhooshC_05}}
 {id=WavePlayer_38, properties {soundWaveAssetPtr=&/MaskMakerBoss_SynthWhooshC_04}}
 {id=WavePlayer_39, properties {soundWaveAssetPtr=&/MaskMakerBoss_SynthWhooshC_03}}
 {id=WavePlayer_4, properties {soundWaveAssetPtr=&/MaskMakerBoss_BeamAttack_HitB_02}}
 {id=WavePlayer_40, properties {soundWaveAssetPtr=&/MaskMakerBoss_SynthWhooshC_02}}
 {id=WavePlayer_41, properties {soundWaveAssetPtr=&/MaskMakerBoss_SynthWhooshC_01}}
 {id=WavePlayer_42, properties {soundWaveAssetPtr=&/MaskMakerBoss_SynthWhooshB_01}}
 {id=WavePlayer_43, properties {soundWaveAssetPtr=&/MaskMakerBoss_SynthWhooshB_02}}
 {id=WavePlayer_44, properties {soundWaveAssetPtr=&/MaskMakerBoss_SynthWhooshB_03}}
 {id=WavePlayer_45, properties {soundWaveAssetPtr=&/MaskMakerBoss_SynthWhooshB_04}}
 {id=WavePlayer_46, properties {soundWaveAssetPtr=&/MaskMakerBoss_SynthWhooshB_05}}
 {id=WavePlayer_47, properties {soundWaveAssetPtr=&/MaskMakerBoss_SynthWhooshB_06}}
 {id=WavePlayer_5, properties {soundWaveAssetPtr=&/MaskMakerBoss_BeamAttack_HitB_03}}
 {id=WavePlayer_51, properties {soundWaveAssetPtr=&/MaskMakerBoss_BeamAttack_HitB_01}}
 {id=WavePlayer_6, properties {soundWaveAssetPtr=&/MaskMakerBoss_BeamAttack_HitB_04}}
 {id=WavePlayer_7, properties {soundWaveAssetPtr=&/MaskMakerBoss_BeamAttack_HitB_05}}
 {id=WavePlayer_8, properties {soundWaveAssetPtr=&/MaskMakerBoss_BeamAttack_HitB_06}}
]
properties {
 attenuationSettings=&/ATT_MaskMakerBoss_Skills
 bOverrideConcurrency=true
 maxDistance=13500
 soundClassObject=&Fort_BossFight_SFX
 virtualizationMode=Disabled
 volumeMultiplier=0.8
 concurrencyOverrides {bVolumeScaleCanRelease=true, maxCount=2, volumeScale=0.15, volumeScaleAttackTime=0.05, volumeScaleReleaseTime=1.5}
}
