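# Sound cue for the MaskMaker boss's ground slam charge-up, routed to the boss fight SFX class with boss skill attenuation.
# A Mixer layers four randomized pools: quiet GiftedTraveler sparks, synth whooshes pitched up slightly at random by a Modulator, boosted slam charges, and reused teleport charges.
# A concurrency override caps simultaneous instances at three, stopping the oldest, and the cue's debug flag is left on.
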SoundCue /MaskMakerBoss/Assets/Audio/SC_MaskMakerBoss_GroundSlam_ChargeUp
root=Mixer_1
nodes [
 {id=Mixer_1, children [Random_6,Modulator_1,Random_1,Random_0], properties {inputVolume [0.35,1,1.7,1]}}
 {id=Modulator_1, children [Random_4], properties {pitchMax=1.2, pitchMin=1}}
 {id=Random_0, children [WavePlayer_1,WavePlayer_2], properties {weights [1,1]}}
 {id=Random_1, children [WavePlayer_0,WavePlayer_3,WavePlayer_4,WavePlayer_5], properties {weights [1,1,1,1]}}
 {
  id=Random_4
  children [WavePlayer_20, WavePlayer_21, WavePlayer_22, WavePlayer_23, WavePlayer_24, WavePlayer_25]
  properties {weights [1,1,1,1,1,1]}
 }
 {
  id=Random_6
  children [
   WavePlayer_15
   WavePlayer_16
   WavePlayer_17
   WavePlayer_18
   WavePlayer_8
   WavePlayer_9
   WavePlayer_10
   WavePlayer_11
   WavePlayer_12
   WavePlayer_13
   WavePlayer_14
  ]
  properties {weights [1,1,1,1,1,1,1,1,1,1,1]}
 }
 {id=WavePlayer_0, properties {soundWaveAssetPtr=&/MaskMakerBoss_GroundSlamCharge_04}}
 {id=WavePlayer_1, properties {soundWaveAssetPtr=&/MaskMakerBoss_TeleportCharge}}
 {id=WavePlayer_10, properties {soundWaveAssetPtr=&/GiftedTraveler_Montage_Sparks_B_03}}
 {id=WavePlayer_11, properties {soundWaveAssetPtr=&/GiftedTraveler_Montage_Sparks_B_04}}
 {id=WavePlayer_12, properties {soundWaveAssetPtr=&/GiftedTraveler_Montage_Sparks_B_05}}
 {id=WavePlayer_13, properties {soundWaveAssetPtr=&/GiftedTraveler_Montage_Sparks_B_06}}
 {id=WavePlayer_14, properties {soundWaveAssetPtr=&/GiftedTraveler_Montage_Sparks_B_07}}
 {id=WavePlayer_15, properties {soundWaveAssetPtr=&/GiftedTraveler_Montage_Sparks_A_01}}
 {id=WavePlayer_16, properties {soundWaveAssetPtr=&/GiftedTraveler_Montage_Sparks_A_02}}
 {id=WavePlayer_17, properties {soundWaveAssetPtr=&/GiftedTraveler_Montage_Sparks_A_03}}
 {id=WavePlayer_18, properties {soundWaveAssetPtr=&/GiftedTraveler_Montage_Sparks_A_04}}
 {id=WavePlayer_2, properties {soundWaveAssetPtr=&/MaskMakerBoss_TeleportChargeA}}
 {id=WavePlayer_20, properties {soundWaveAssetPtr=&/MaskMakerBoss_SynthWhooshA_01}}
 {id=WavePlayer_21, properties {soundWaveAssetPtr=&/MaskMakerBoss_SynthWhooshA_02}}
 {id=WavePlayer_22, properties {soundWaveAssetPtr=&/MaskMakerBoss_SynthWhooshA_03}}
 {id=WavePlayer_23, properties {soundWaveAssetPtr=&/MaskMakerBoss_SynthWhooshA_04}}
 {id=WavePlayer_24, properties {soundWaveAssetPtr=&/MaskMakerBoss_SynthWhooshA_05}}
 {id=WavePlayer_25, properties {soundWaveAssetPtr=&/MaskMakerBoss_SynthWhooshA_06}}
 {id=WavePlayer_3, properties {soundWaveAssetPtr=&/MaskMakerBoss_GroundSlamCharge_03}}
 {id=WavePlayer_4, properties {soundWaveAssetPtr=&/MaskMakerBoss_GroundSlamCharge_02}}
 {id=WavePlayer_5, properties {soundWaveAssetPtr=&/MaskMakerBoss_GroundSlamCharge_01}}
 {id=WavePlayer_8, properties {soundWaveAssetPtr=&/GiftedTraveler_Montage_Sparks_B_01}}
 {id=WavePlayer_9, properties {soundWaveAssetPtr=&/GiftedTraveler_Montage_Sparks_B_02}}
]
properties {
 attenuationSettings=&/ATT_MaskMakerBoss_Skills
 bDebug=true
 bOverrideConcurrency=true
 maxDistance=13500
 soundClassObject=&Fort_BossFight_SFX
 volumeMultiplier=1
 concurrencyOverrides {maxCount=3, resolutionRule=StopOldest}
}
