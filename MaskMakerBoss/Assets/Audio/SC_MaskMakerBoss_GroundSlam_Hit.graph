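# Sound cue for the impact of the Mask Maker boss's ground slam, routed to the boss fight SFX class with boss-skill attenuation.
# A Mixer layers four Random nodes at separate volumes, each picking one equally weighted variant of beam-attack hit, slam debris, sparks or ground slam.
# The hit and sparks layers reuse waves named for the beam attack and Gifted Traveler montage.
# A concurrency override caps simultaneous plays, stopping the oldest.
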SoundCue /MaskMakerBoss/Assets/Audio/SC_MaskMakerBoss_GroundSlam_Hit
root=Mixer_6
nodes [
 {id=Mixer_6, children [Random_5,Random_13,Random_3,Random_2], properties {inputVolume [1.25,1.5,0.5,1]}}
 {id=Random_13, children [WavePlayer_44, WavePlayer_45, WavePlayer_46], properties {weights [1,1,1]}}
 {id=Random_2, children [WavePlayer_7, WavePlayer_8, WavePlayer_9, WavePlayer_10], properties {weights [1,1,1,1]}}
 {id=Random_3, children [WavePlayer_15, WavePlayer_16, WavePlayer_17, WavePlayer_18, WavePlayer_19], properties {weights [1,1,1,1,1]}}
 {
  id=Random_5
  children [WavePlayer_26, WavePlayer_27, WavePlayer_28, WavePlayer_29, WavePlayer_30, WavePlayer_31]
  properties {weights [1,1,1,1,1,1]}
 }
 {id=WavePlayer_10, properties {soundWaveAssetPtr=&/MaskMakerBoss_GroundSlam_04}}
 {id=WavePlayer_15, properties {soundWaveAssetPtr=&/GiftedTraveler_Montage_Sparks_C_01}}
 {id=WavePlayer_16, properties {soundWaveAssetPtr=&/GiftedTraveler_Montage_Sparks_C_02}}
 {id=WavePlayer_17, properties {soundWaveAssetPtr=&/GiftedTraveler_Montage_Sparks_C_03}}
 {id=WavePlayer_18, properties {soundWaveAssetPtr=&/GiftedTraveler_Montage_Sparks_C_04}}
 {id=WavePlayer_19, properties {soundWaveAssetPtr=&/GiftedTraveler_Montage_Sparks_C_05}}
 {id=WavePlayer_26, properties {soundWaveAssetPtr=&/MaskMakerBoss_BeamAttack_HitB_06}}
 {id=WavePlayer_27, properties {soundWaveAssetPtr=&/MaskMakerBoss_BeamAttack_HitB_05}}
 {id=WavePlayer_28, properties {soundWaveAssetPtr=&/MaskMakerBoss_BeamAttack_HitB_04}}
 {id=WavePlayer_29, properties {soundWaveAssetPtr=&/MaskMakerBoss_BeamAttack_HitB_03}}
 {id=WavePlayer_30, properties {soundWaveAssetPtr=&/MaskMakerBoss_BeamAttack_HitB_02}}
 {id=WavePlayer_31, properties {soundWaveAssetPtr=&/MaskMakerBoss_BeamAttack_HitB_01}}
 {id=WavePlayer_44, properties {soundWaveAssetPtr=&/MaskMakerBoss_SlamDebris_01}}
 {id=WavePlayer_45, properties {soundWaveAssetPtr=&/MaskMakerBoss_SlamDebris_02}}
 {id=WavePlayer_46, properties {soundWaveAssetPtr=&/MaskMakerBoss_SlamDebris_03}}
 {id=WavePlayer_7, properties {soundWaveAssetPtr=&/MaskMakerBoss_GroundSlam_01}}
 {id=WavePlayer_8, properties {soundWaveAssetPtr=&/MaskMakerBoss_GroundSlam_02}}
 {id=WavePlayer_9, properties {soundWaveAssetPtr=&/MaskMakerBoss_GroundSlam_03}}
]
properties {
 attenuationSettings=&/ATT_MaskMakerBoss_Skills
 bDebug=true
 bOverrideConcurrency=true
 maxDistance=13500
 soundClassObject=&Fort_BossFight_SFX
 volumeMultiplier=1
 concurrencyOverrides {maxCount=3, resolutionRule=StopOldest}
}
