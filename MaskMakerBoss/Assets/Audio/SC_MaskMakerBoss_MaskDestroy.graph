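# Sound cue for the MaskMaker boss's mask-destroy sound effect, played as a boss fight SFX.
# A modulator feeds a quality switch whose mixers layer a random mask-destroy sound with a random beam-attack hit.
# The first quality level draws from larger sample pools, while the other three share a mixer with two choices per layer.
# It uses mask-hit attenuation, limits concurrency by stopping the oldest, and has distance-based sends to music and weapon sidechain buses.
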SoundCue /MaskMakerBoss/Assets/Audio/SC_MaskMakerBoss_MaskDestroy
root=Modulator_0
nodes [
 {id=Mixer_1, children [Random_6,Random_3], properties {inputVolume [1,1]}}
 {id=Mixer_2, children [Random_2,Random_1], properties {inputVolume [1,1]}}
 {id=Modulator_0, children [QualityLevel_0]}
 {id=QualityLevel_0, children [Mixer_2,Mixer_1,Mixer_1,Mixer_1]}
 {id=Random_1, children [WavePlayer_3,WavePlayer_4,WavePlayer_5,WavePlayer_6,WavePlayer_0,WavePlayer_1], properties {weights [1,1,1,1,1,1]}}
 {id=Random_2, children [WavePlayer_7,WavePlayer_8,WavePlayer_9], properties {weights [1,1,1]}}
 {id=Random_3, children [WavePlayer_2, WavePlayer_12], properties {weights [1,1]}}
 {id=Random_6, children [WavePlayer_10, WavePlayer_11], properties {weights [1,1]}}
 {id=WavePlayer_0, properties {soundWaveAssetPtr=&/MaskMakerBoss_BeamAttack_HitB_05}}
 {id=WavePlayer_1, properties {soundWaveAssetPtr=&/MaskMakerBoss_BeamAttack_HitB_06}}
 {id=WavePlayer_10, properties {soundWaveAssetPtr=&/MaskMakerBoss_MaskDestroy_03}}
 {id=WavePlayer_11, properties {soundWaveAssetPtr=&/MaskMakerBoss_MaskDestroy_02}}
 {id=WavePlayer_12, properties {soundWaveAssetPtr=&/MaskMakerBoss_BeamAttack_HitB_02}}
 {id=WavePlayer_2, properties {soundWaveAssetPtr=&/MaskMakerBoss_BeamAttack_HitB_01}}
 {id=WavePlayer_3, properties {soundWaveAssetPtr=&MaskMakerBoss_BeamAttack_HitB_01}}
 {id=WavePlayer_4, properties {soundWaveAssetPtr=&MaskMakerBoss_BeamAttack_HitB_02}}
 {id=WavePlayer_5, properties {soundWaveAssetPtr=&/MaskMakerBoss_BeamAttack_HitB_03}}
 {id=WavePlayer_6, properties {soundWaveAssetPtr=&/MaskMakerBoss_BeamAttack_HitB_04}}
 {id=WavePlayer_7, properties {soundWaveAssetPtr=&MaskMakerBoss_MaskDestroy_03}}
 {id=WavePlayer_8, properties {soundWaveAssetPtr=&MaskMakerBoss_MaskDestroy_02}}
 {id=WavePlayer_9, properties {soundWaveAssetPtr=&/MaskMakerBoss_MaskDestroy_01}}
]
properties {
 attenuationSettings=&/ATT_MaskMakerBoss_MaskHit
 bOverrideConcurrency=true
 maxDistance=2100
 soundClassObject=&Fort_BossFight_SFX
 volumeMultiplier=0.65
 busSends [
  {
   audioBus=&AB_Sidechain_FortMusic_Vol
   maxSendDistance=6000
   maxSendLevel=0
   minSendDistance=1000
   minSendLevel=1
   sendLevel=0
   soundSourceBus=null
   sourceBusSendLevelControlMethod=Linear
  }
  {
   audioBus=&AB_Sidechain_FortWeapon_Vol
   maxSendDistance=5000
   maxSendLevel=0
   minSendDistance=2500
   minSendLevel=1
   sendLevel=0
   soundSourceBus=null
   sourceBusSendLevelControlMethod=Linear
  }
 ]
 concurrencyOverrides {maxCount=2, resolutionRule=StopOldest}
}
