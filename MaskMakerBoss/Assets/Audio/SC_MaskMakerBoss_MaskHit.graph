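# Sound cue for the MaskMaker boss's mask hit impact, played as boss-fight SFX with its own attenuation.
# A quality switch picks randomly from six impact waves at the first level or two at the other levels, then a modulator slightly varies pitch upward.
# Distance-scaled sends feed music and weapon volume sidechain buses, strongest up close.
# A concurrency override caps it at two instances, stopping the oldest.
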SoundCue /MaskMakerBoss/Assets/Audio/SC_MaskMakerBoss_MaskHit
root=Modulator_0
nodes [
 {id=Modulator_0, children [QualityLevel_0], properties {pitchMax=1.1, pitchMin=1}}
 {id=QualityLevel_0, children [Random_2,Random_4,Random_4,Random_4]}
 {
  id=Random_2
  children [WavePlayer_14, WavePlayer_15, WavePlayer_16, WavePlayer_17, WavePlayer_18, WavePlayer_19]
  properties {weights [1,1,1,1,1,1]}
 }
 {id=Random_4, children [WavePlayer_20, WavePlayer_21], properties {weights [1,1]}}
 {id=WavePlayer_14, properties {soundWaveAssetPtr=&/MaskMaker_MaskImpact_06}}
 {id=WavePlayer_15, properties {soundWaveAssetPtr=&/MaskMaker_MaskImpact_05}}
 {id=WavePlayer_16, properties {soundWaveAssetPtr=&/MaskMaker_MaskImpact_04}}
 {id=WavePlayer_17, properties {soundWaveAssetPtr=&/MaskMaker_MaskImpact_03}}
 {id=WavePlayer_18, properties {soundWaveAssetPtr=&/MaskMaker_MaskImpact_02}}
 {id=WavePlayer_19, properties {soundWaveAssetPtr=&/MaskMaker_MaskImpact_01}}
 {id=WavePlayer_20, properties {soundWaveAssetPtr=&MaskMaker_MaskImpact_02}}
 {id=WavePlayer_21, properties {soundWaveAssetPtr=&MaskMaker_MaskImpact_01}}
]
properties {
 attenuationSettings=&/ATT_MaskMakerBoss_MaskHit
 bOverrideConcurrency=true
 maxDistance=2100
 soundClassObject=&Fort_BossFight_SFX
 volumeMultiplier=2.2
 busSends [
  {
   audioBus=&AB_Sidechain_FortMusic_Vol
   maxSendDistance=6000
   maxSendLevel=0
   minSendDistance=1000
   minSendLevel=1
   sendLevel=0
   soundSourceBus=null
   sourceBusSendLevelControlMethod=Linear
  }
  {
   audioBus=&AB_Sidechain_FortWeapon_Vol
   maxSendDistance=5000
   maxSendLevel=0
   minSendDistance=2500
   minSendLevel=1
   sendLevel=0
   soundSourceBus=null
   sourceBusSendLevelControlMethod=Linear
  }
 ]
 concurrencyOverrides {maxCount=2, resolutionRule=StopOldest}
}
