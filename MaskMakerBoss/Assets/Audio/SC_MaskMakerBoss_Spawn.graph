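# Sound cue for the Mask Maker boss spawn, layering randomized spawn hits with synth whooshes.
# A Mixer plays five equally weighted random layers at once: two spawn-hit pairs and three whoosh sets, with the A whoosh set starting after a short fixed delay.
# It routes to the boss-fight SFX class, uses the boss skills attenuation asset, and caps simultaneous instances through a concurrency override.
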SoundCue /MaskMakerBoss/Assets/Audio/SC_MaskMakerBoss_Spawn
root=Mixer_0
nodes [
 {id=Delay_0, children [Random_4], properties {delayMax=0.2, delayMin=0.2}}
 {id=Mixer_0, children [Random_0,Random_6,Delay_0,Random_7,Random_1], properties {inputVolume [1,1,1,1,1]}}
 {id=Random_0, children [WavePlayer_13, WavePlayer_14], properties {weights [1,1]}}
 {id=Random_1, children [WavePlayer_0,WavePlayer_1], properties {weights [1,1]}}
 {id=Random_4, children [WavePlayer_21, WavePlayer_22, WavePlayer_24], properties {weights [1,1,1]}}
 {
  id=Random_6
  children [WavePlayer_32, WavePlayer_33, WavePlayer_34, WavePlayer_35, WavePlayer_36, WavePlayer_37]
  properties {weights [1,1,1,1,1,1]}
 }
 {id=Random_7, children [WavePlayer_4,WavePlayer_5,WavePlayer_6,WavePlayer_7,WavePlayer_8,WavePlayer_9], properties {weights [1,1,1,1,1,1]}}
 {id=WavePlayer_0, properties {soundWaveAssetPtr=&/MaskMakerBoss_Spawn_D_01}}
 {id=WavePlayer_1, properties {soundWaveAssetPtr=&/MaskMakerBoss_Spawn_D_02}}
 {id=WavePlayer_13, properties {soundWaveAssetPtr=&/MaskMakerBoss_Spawn_B}}
 {id=WavePlayer_14, properties {soundWaveAssetPtr=&/MaskMakerBoss_Spawn_C}}
 {id=WavePlayer_21, properties {soundWaveAssetPtr=&/MaskMakerBoss_SynthWhooshA_02}}
 {id=WavePlayer_22, properties {soundWaveAssetPtr=&/MaskMakerBoss_SynthWhooshA_03}}
 {id=WavePlayer_24, properties {soundWaveAssetPtr=&/MaskMakerBoss_SynthWhooshA_05}}
 {id=WavePlayer_32, properties {soundWaveAssetPtr=&/MaskMakerBoss_SynthWhooshB_06}}
 {id=WavePlayer_33, properties {soundWaveAssetPtr=&/MaskMakerBoss_SynthWhooshB_05}}
 {id=WavePlayer_34, properties {soundWaveAssetPtr=&/MaskMakerBoss_SynthWhooshB_04}}
 {id=WavePlayer_35, properties {soundWaveAssetPtr=&/MaskMakerBoss_SynthWhooshB_03}}
 {id=WavePlayer_36, properties {soundWaveAssetPtr=&/MaskMakerBoss_SynthWhooshB_02}}
 {id=WavePlayer_37, properties {soundWaveAssetPtr=&/MaskMakerBoss_SynthWhooshB_01}}
 {id=WavePlayer_4, properties {soundWaveAssetPtr=&/MaskMakerBoss_SynthWhooshD_06}}
 {id=WavePlayer_5, properties {soundWaveAssetPtr=&/MaskMakerBoss_SynthWhooshD_05}}
 {id=WavePlayer_6, properties {soundWaveAssetPtr=&/MaskMakerBoss_SynthWhooshD_04}}
 {id=WavePlayer_7, properties {soundWaveAssetPtr=&/MaskMakerBoss_SynthWhooshD_03}}
 {id=WavePlayer_8, properties {soundWaveAssetPtr=&/MaskMakerBoss_SynthWhooshD_02}}
 {id=WavePlayer_9, properties {soundWaveAssetPtr=&/MaskMakerBoss_SynthWhooshD_01}}
]
properties {
 attenuationSettings=&/ATT_MaskMakerBoss_Skills
 bOverrideConcurrency=true
 maxDistance=13500
 soundClassObject=&Fort_BossFight_SFX
 volumeMultiplier=0.8
 attenuationOverrides {falloffDistance=2000, stereoSpread=500, attenuationShapeExtents {x=25000}}
 concurrencyOverrides {maxCount=3}
}
