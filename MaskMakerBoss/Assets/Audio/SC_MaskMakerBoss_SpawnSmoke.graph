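# Sound cue for the MaskMaker boss's smoke spawn effect.
# A mixer layers a randomly picked stun smoke sample, boosted in volume, over one of six randomly chosen synth whoosh variations, with equal weights in each pool.
# Playback uses the boss skills attenuation and boss fight SFX sound class, with a concurrency override that limits simultaneous instances.
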SoundCue /MaskMakerBoss/Assets/Audio/SC_MaskMakerBoss_SpawnSmoke
root=Mixer_1
nodes [
 {id=Mixer_1, children [Random_2,Random_1], properties {inputVolume [1.75,1]}}
 {
  id=Random_1
  children [WavePlayer_2, WavePlayer_3, WavePlayer_10, WavePlayer_11, WavePlayer_12, WavePlayer_15]
  properties {weights [1,1,1,1,1,1]}
 }
 {id=Random_2, children [WavePlayer_0,WavePlayer_1], properties {weights [1,1]}}
 {id=WavePlayer_0, properties {soundWaveAssetPtr=&/MaskMakerBoss_StunSmoke_01}}
 {id=WavePlayer_1, properties {soundWaveAssetPtr=&/MaskMakerBoss_StunSmoke_02}}
 {id=WavePlayer_10, properties {soundWaveAssetPtr=&/MaskMakerBoss_SynthWhooshC_04}}
 {id=WavePlayer_11, properties {soundWaveAssetPtr=&/MaskMakerBoss_SynthWhooshC_03}}
 {id=WavePlayer_12, properties {soundWaveAssetPtr=&/MaskMakerBoss_SynthWhooshC_02}}
 {id=WavePlayer_15, properties {soundWaveAssetPtr=&/MaskMakerBoss_SynthWhooshC_01}}
 {id=WavePlayer_2, properties {soundWaveAssetPtr=&/MaskMakerBoss_SynthWhooshC_06}}
 {id=WavePlayer_3, properties {soundWaveAssetPtr=&/MaskMakerBoss_SynthWhooshC_05}}
]
properties {
 attenuationSettings=&/ATT_MaskMakerBoss_Skills
 bOverrideConcurrency=true
 maxDistance=13500
 soundClassObject=&Fort_BossFight_SFX
 volumeMultiplier=0.55
 attenuationOverrides {falloffDistance=2000, stereoSpread=500, attenuationShapeExtents {x=25000}}
 concurrencyOverrides {maxCount=3}
}
