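# Sound cue for the build sound of the MaskMaker boss's SpireShards, classed as boss-fight SFX.
# It picks one of three equally weighted wave variations and passes it through a modulator.
# It uses the boss skills attenuation and overrides concurrency to cap simultaneous instances.
# Distance-scaled bus sends feed the music and weapon sidechain volume buses, strongest up close and fading linearly to zero.
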SoundCue /MaskMakerBoss/Assets/Audio/SC_MaskMakerBoss_SpireShards_Build
root=Modulator_1
nodes [
 {id=Modulator_1, children [Random_1]}
 {id=Random_1, children [WavePlayer_0,WavePlayer_1,WavePlayer_2], properties {weights [1,1,1]}}
 {id=WavePlayer_0, properties {soundWaveAssetPtr=&/MaskMaker_SpireShards_Build_03}}
 {id=WavePlayer_1, properties {soundWaveAssetPtr=&/MaskMaker_SpireShards_Build_02}}
 {id=WavePlayer_2, properties {soundWaveAssetPtr=&/MaskMaker_SpireShards_Build_01}}
]
properties {
 attenuationSettings=&/ATT_MaskMakerBoss_Skills
 bOverrideConcurrency=true
 maxDistance=13500
 soundClassObject=&Fort_BossFight_SFX
 volumeMultiplier=0.8
 busSends [
  {
   audioBus=&AB_Sidechain_FortMusic_Vol
   maxSendDistance=2500
   maxSendLevel=0
   minSendDistance=0
   minSendLevel=1
   sendLevel=0
   soundSourceBus=null
   sourceBusSendLevelControlMethod=Linear
  }
  {
   audioBus=&AB_Sidechain_FortWeapon_Vol
   maxSendDistance=2500
   maxSendLevel=0
   minSendDistance=0
   minSendLevel=1
   sendLevel=0
   soundSourceBus=null
   sourceBusSendLevelControlMethod=Linear
  }
 ]
 concurrencyOverrides {maxCount=6}
}
