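# Sound cue for the Mask Maker boss's stunned sound, layered from synth whooshes, beam-attack hits and sparks reused from Gifted Traveler montage audio.
# A mixer blends five randomized variant pools at differing volumes under a root modulator.
# A beam hit, whoosh A and sparks play immediately, while whooshes C and B follow after fixed delays of about half a second or more to stagger the stun.
# It uses the boss skills attenuation asset, the boss-fight SFX class and a concurrency override.
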SoundCue /MaskMakerBoss/Assets/Audio/SC_MaskMakerBoss_Stunned
root=Modulator_0
nodes [
 {id=Delay_1, children [Random_6], properties {delayMax=0.65, delayMin=0.65}}
 {id=Delay_2, children [Random_2], properties {delayMax=0.55, delayMin=0.55}}
 {id=Mixer_0, children [Delay_1,Random_4,Random_5,Delay_2,Random_0], properties {inputVolume [1,1.25,0.5,0.75,1]}}
 {id=Modulator_0, children [Mixer_0]}
 {id=Random_0, children [WavePlayer_1,WavePlayer_2,WavePlayer_3,WavePlayer_4,WavePlayer_5,WavePlayer_6], properties {weights [1,1,1,1,1,1]}}
 {
  id=Random_2
  children [WavePlayer_10, WavePlayer_11, WavePlayer_12, WavePlayer_15, WavePlayer_16, WavePlayer_17]
  properties {weights [1,1,1,1,1,1]}
 }
 {id=Random_4, children [WavePlayer_21, WavePlayer_22, WavePlayer_24], properties {weights [1,1,1]}}
 {id=Random_5, children [WavePlayer_26, WavePlayer_28, WavePlayer_30, WavePlayer_31, WavePlayer_0], properties {weights [1,1,1,1,1]}}
 {
  id=Random_6
  children [WavePlayer_32, WavePlayer_33, WavePlayer_34, WavePlayer_35, WavePlayer_36, WavePlayer_37]
  properties {weights [1,1,1,1,1,1]}
 }
 {id=WavePlayer_0, properties {soundWaveAssetPtr=&/GiftedTraveler_Montage_Sparks_B_07}}
 {id=WavePlayer_1, properties {soundWaveAssetPtr=&/MaskMakerBoss_BeamAttack_HitB_01}}
 {id=WavePlayer_10, properties {soundWaveAssetPtr=&/MaskMakerBoss_SynthWhooshC_06}}
 {id=WavePlayer_11, properties {soundWaveAssetPtr=&/MaskMakerBoss_SynthWhooshC_05}}
 {id=WavePlayer_12, properties {soundWaveAssetPtr=&/MaskMakerBoss_SynthWhooshC_04}}
 {id=WavePlayer_15, properties {soundWaveAssetPtr=&/MaskMakerBoss_SynthWhooshC_03}}
 {id=WavePlayer_16, properties {soundWaveAssetPtr=&/MaskMakerBoss_SynthWhooshC_02}}
 {id=WavePlayer_17, properties {soundWaveAssetPtr=&/MaskMakerBoss_SynthWhooshC_01}}
 {id=WavePlayer_2, properties {soundWaveAssetPtr=&/MaskMakerBoss_BeamAttack_HitB_02}}
 {id=WavePlayer_21, properties {soundWaveAssetPtr=&/MaskMakerBoss_SynthWhooshA_02}}
 {id=WavePlayer_22, properties {soundWaveAssetPtr=&/MaskMakerBoss_SynthWhooshA_03}}
 {id=WavePlayer_24, properties {soundWaveAssetPtr=&/MaskMakerBoss_SynthWhooshA_05}}
 {id=WavePlayer_26, properties {soundWaveAssetPtr=&/GiftedTraveler_Montage_Sparks_B_01}}
 {id=WavePlayer_28, properties {soundWaveAssetPtr=&/GiftedTraveler_Montage_Sparks_B_03}}
 {id=WavePlayer_3, properties {soundWaveAssetPtr=&/MaskMakerBoss_BeamAttack_HitB_03}}
 {id=WavePlayer_30, properties {soundWaveAssetPtr=&/GiftedTraveler_Montage_Sparks_B_05}}
 {id=WavePlayer_31, properties {soundWaveAssetPtr=&/GiftedTraveler_Montage_Sparks_B_06}}
 {id=WavePlayer_32, properties {soundWaveAssetPtr=&/MaskMakerBoss_SynthWhooshB_06}}
 {id=WavePlayer_33, properties {soundWaveAssetPtr=&/MaskMakerBoss_SynthWhooshB_05}}
 {id=WavePlayer_34, properties {soundWaveAssetPtr=&/MaskMakerBoss_SynthWhooshB_04}}
 {id=WavePlayer_35, properties {soundWaveAssetPtr=&/MaskMakerBoss_SynthWhooshB_03}}
 {id=WavePlayer_36, properties {soundWaveAssetPtr=&/MaskMakerBoss_SynthWhooshB_02}}
 {id=WavePlayer_37, properties {soundWaveAssetPtr=&/MaskMakerBoss_SynthWhooshB_01}}
 {id=WavePlayer_4, properties {soundWaveAssetPtr=&/MaskMakerBoss_BeamAttack_HitB_04}}
 {id=WavePlayer_5, properties {soundWaveAssetPtr=&/MaskMakerBoss_BeamAttack_HitB_05}}
 {id=WavePlayer_6, properties {soundWaveAssetPtr=&/MaskMakerBoss_BeamAttack_HitB_06}}
]
properties {
 attenuationSettings=&/ATT_MaskMakerBoss_Skills
 bOverrideConcurrency=true
 maxDistance=13500
 soundClassObject=&Fort_BossFight_SFX
 volumeMultiplier=1.2
 attenuationOverrides {falloffDistance=2000, stereoSpread=500, attenuationShapeExtents {x=25000}}
 concurrencyOverrides {maxCount=3}
}
