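# Sound cue for the MaskMaker boss's teleport charge-up, classed as boss-fight SFX.
# A root Mixer layers three Random nodes: synth whoosh A, synth whoosh C and braam variations, with the two whoosh layers boosted over the braam.
# Each layer picks evenly among its waves.
# It uses the boss skills attenuation and overrides concurrency to cap simultaneous instances, stopping the oldest.
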SoundCue /MaskMakerBoss/Assets/Audio/SC_MaskMakerBoss_Teleport_ChargeUp
root=Mixer_1
nodes [
 {id=Mixer_1, children [Random_4,Random_2,Random_3], properties {inputVolume [1.2,1.2,1]}}
 {
  id=Random_2
  children [WavePlayer_5, WavePlayer_8, WavePlayer_9, WavePlayer_10, WavePlayer_11, WavePlayer_12]
  properties {weights [1,1,1,1,1,1]}
 }
 {id=Random_3, children [WavePlayer_16, WavePlayer_17, WavePlayer_18, WavePlayer_19], properties {weights [1,1,1,1]}}
 {
  id=Random_4
  children [WavePlayer_20, WavePlayer_21, WavePlayer_22, WavePlayer_23, WavePlayer_24, WavePlayer_25]
  properties {weights [1,1,1,1,1,1]}
 }
 {id=WavePlayer_10, properties {soundWaveAssetPtr=&/MaskMakerBoss_SynthWhooshC_04}}
 {id=WavePlayer_11, properties {soundWaveAssetPtr=&/MaskMakerBoss_SynthWhooshC_05}}
 {id=WavePlayer_12, properties {soundWaveAssetPtr=&/MaskMakerBoss_SynthWhooshC_06}}
 {id=WavePlayer_16, properties {soundWaveAssetPtr=&/MaskMakerBoss_Braam_01}}
 {id=WavePlayer_17, properties {soundWaveAssetPtr=&/MaskMakerBoss_Braam_02}}
 {id=WavePlayer_18, properties {soundWaveAssetPtr=&/MaskMakerBoss_Braam_03}}
 {id=WavePlayer_19, properties {soundWaveAssetPtr=&/MaskMakerBoss_Braam_04}}
 {id=WavePlayer_20, properties {soundWaveAssetPtr=&/MaskMakerBoss_SynthWhooshA_01}}
 {id=WavePlayer_21, properties {soundWaveAssetPtr=&/MaskMakerBoss_SynthWhooshA_02}}
 {id=WavePlayer_22, properties {soundWaveAssetPtr=&/MaskMakerBoss_SynthWhooshA_03}}
 {id=WavePlayer_23, properties {soundWaveAssetPtr=&/MaskMakerBoss_SynthWhooshA_04}}
 {id=WavePlayer_24, properties {soundWaveAssetPtr=&/MaskMakerBoss_SynthWhooshA_05}}
 {id=WavePlayer_25, properties {soundWaveAssetPtr=&/MaskMakerBoss_SynthWhooshA_06}}
 {id=WavePlayer_5, properties {soundWaveAssetPtr=&/MaskMakerBoss_SynthWhooshC_01}}
 {id=WavePlayer_8, properties {soundWaveAssetPtr=&/MaskMakerBoss_SynthWhooshC_02}}
 {id=WavePlayer_9, properties {soundWaveAssetPtr=&/MaskMakerBoss_SynthWhooshC_03}}
]
properties {
 attenuationSettings=&/ATT_MaskMakerBoss_Skills
 bDebug=true
 bOverrideConcurrency=true
 maxDistance=13500
 soundClassObject=&Fort_BossFight_SFX
 volumeMultiplier=0.95
 concurrencyOverrides {maxCount=3, resolutionRule=StopOldest}
}
