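# Sound cue for the MaskMakerBoss teleport slam, played as boss-fight SFX with the boss skills attenuation and a capped voice count.
# A root modulator feeds a quality switch whose mixers layer random picks of sparks, two beam-attack hit sets and slam waves, with one hit set boosted.
# The first quality level uses the full variation pools, while the other three share a two-variant mixer.
# Distance-scaled sends feed music and weapon sidechain buses.
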SoundCue /MaskMakerBoss/Assets/Audio/SC_MaskMakerBoss_Teleport_Slam
root=Modulator_1
nodes [
 {id=Mixer_0, children [Random_3,Random_2,Random_1,Random_0], properties {inputVolume [0.5,1.5,1,1]}}
 {id=Mixer_3, children [Random_4,Random_10,Random_7,Random_6], properties {inputVolume [0.5,1.5,1,1]}}
 {id=Modulator_1, children [QualityLevel_0]}
 {id=QualityLevel_0, children [Mixer_0,Mixer_3,Mixer_3,Mixer_3]}
 {
  id=Random_0
  children [WavePlayer_26, WavePlayer_27, WavePlayer_28, WavePlayer_29, WavePlayer_32, WavePlayer_33]
  properties {weights [1,1,1,1,1,1]}
 }
 {
  id=Random_1
  children [WavePlayer_10, WavePlayer_11, WavePlayer_12, WavePlayer_13, WavePlayer_24, WavePlayer_25]
  properties {weights [1,1,1,1,1,1]}
 }
 {id=Random_10, children [WavePlayer_30, WavePlayer_31], properties {weights [1,1]}}
 {id=Random_2, children [WavePlayer_0,WavePlayer_1,WavePlayer_2,WavePlayer_3,WavePlayer_4,WavePlayer_5], properties {weights [1,1,1,1,1,1]}}
 {id=Random_3, children [WavePlayer_15, WavePlayer_16, WavePlayer_17, WavePlayer_18, WavePlayer_19], properties {weights [1,1,1,1,1]}}
 {id=Random_4, children [WavePlayer_6,WavePlayer_7], properties {weights [1,1]}}
 {id=Random_6, children [WavePlayer_36, WavePlayer_37], properties {weights [1,1]}}
 {id=Random_7, children [WavePlayer_34, WavePlayer_35], properties {weights [1,1]}}
 {id=WavePlayer_0, properties {soundWaveAssetPtr=&/MaskMakerBoss_BeamAttack_HitB_06}}
 {id=WavePlayer_1, properties {soundWaveAssetPtr=&/MaskMakerBoss_BeamAttack_HitB_05}}
 {id=WavePlayer_10, properties {soundWaveAssetPtr=&/MaskMakerBoss_BeamAttack_HitA_01}}
 {id=WavePlayer_11, properties {soundWaveAssetPtr=&/MaskMakerBoss_BeamAttack_HitA_02}}
 {id=WavePlayer_12, properties {soundWaveAssetPtr=&/MaskMakerBoss_BeamAttack_HitA_03}}
 {id=WavePlayer_13, properties {soundWaveAssetPtr=&/MaskMakerBoss_BeamAttack_HitA_04}}
 {id=WavePlayer_15, properties {soundWaveAssetPtr=&/GiftedTraveler_Montage_Sparks_C_01}}
 {id=WavePlayer_16, properties {soundWaveAssetPtr=&/GiftedTraveler_Montage_Sparks_C_02}}
 {id=WavePlayer_17, properties {soundWaveAssetPtr=&/GiftedTraveler_Montage_Sparks_C_03}}
 {id=WavePlayer_18, properties {soundWaveAssetPtr=&/GiftedTraveler_Montage_Sparks_C_04}}
 {id=WavePlayer_19, properties {soundWaveAssetPtr=&/GiftedTraveler_Montage_Sparks_C_05}}
 {id=WavePlayer_2, properties {soundWaveAssetPtr=&/MaskMakerBoss_BeamAttack_HitB_04}}
 {id=WavePlayer_24, properties {soundWaveAssetPtr=&/MaskMakerBoss_BeamAttack_HitA_05}}
 {id=WavePlayer_25, properties {soundWaveAssetPtr=&/MaskMakerBoss_BeamAttack_HitA_06}}
 {id=WavePlayer_26, properties {soundWaveAssetPtr=&/MaskMakerBoss_TeleportSlam_06}}
 {id=WavePlayer_27, properties {soundWaveAssetPtr=&/MaskMakerBoss_TeleportSlam_05}}
 {id=WavePlayer_28, properties {soundWaveAssetPtr=&/MaskMakerBoss_TeleportSlam_04}}
 {id=WavePlayer_29, properties {soundWaveAssetPtr=&/MaskMakerBoss_TeleportSlam_03}}
 {id=WavePlayer_3, properties {soundWaveAssetPtr=&/MaskMakerBoss_BeamAttack_HitB_03}}
 {id=WavePlayer_30, properties {soundWaveAssetPtr=&MaskMakerBoss_BeamAttack_HitB_06}}
 {id=WavePlayer_31, properties {soundWaveAssetPtr=&MaskMakerBoss_BeamAttack_HitB_05}}
 {id=WavePlayer_32, properties {soundWaveAssetPtr=&/MaskMakerBoss_TeleportSlam_02}}
 {id=WavePlayer_33, properties {soundWaveAssetPtr=&/MaskMakerBoss_TeleportSlam_01}}
 {id=WavePlayer_34, properties {soundWaveAssetPtr=&MaskMakerBoss_BeamAttack_HitA_01}}
 {id=WavePlayer_35, properties {soundWaveAssetPtr=&MaskMakerBoss_BeamAttack_HitA_02}}
 {id=WavePlayer_36, properties {soundWaveAssetPtr=&MaskMakerBoss_TeleportSlam_06}}
 {id=WavePlayer_37, properties {soundWaveAssetPtr=&MaskMakerBoss_TeleportSlam_05}}
 {id=WavePlayer_4, properties {soundWaveAssetPtr=&/MaskMakerBoss_BeamAttack_HitB_02}}
 {id=WavePlayer_5, properties {soundWaveAssetPtr=&/MaskMakerBoss_BeamAttack_HitB_01}}
 {id=WavePlayer_6, properties {soundWaveAssetPtr=&GiftedTraveler_Montage_Sparks_C_01}}
 {id=WavePlayer_7, properties {soundWaveAssetPtr=&GiftedTraveler_Montage_Sparks_C_02}}
]
properties {
 attenuationSettings=&/ATT_MaskMakerBoss_Skills
 bOverrideConcurrency=true
 maxDistance=13500
 soundClassObject=&Fort_BossFight_SFX
 volumeMultiplier=0.8
 busSends [
  {
   audioBus=&AB_Sidechain_FortMusic_Vol
   maxSendDistance=6000
   maxSendLevel=0
   minSendDistance=1000
   minSendLevel=1
   sendLevel=0
   soundSourceBus=null
   sourceBusSendLevelControlMethod=Linear
  }
  {
   audioBus=&AB_Sidechain_FortWeapon_Vol
   maxSendDistance=5000
   maxSendLevel=0
   minSendDistance=2500
   minSendLevel=1
   sendLevel=0
   soundSourceBus=null
   sourceBusSendLevelControlMethod=Linear
  }
 ]
 concurrencyOverrides {maxCount=3}
}
