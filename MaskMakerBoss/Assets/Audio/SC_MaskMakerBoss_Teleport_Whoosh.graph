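# Sound cue for the MaskMakerBoss teleport whoosh, classed as boss-fight SFX.
# It mixes a random synth whoosh at full level with a quieter random spark layer, borrowed from GiftedTraveler montage waves, that starts after a short fixed delay.
# A modulator at the root wraps the mix.
# It uses the boss skills attenuation and overrides concurrency to allow a few instances, stopping the oldest, with debug enabled.
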SoundCue /MaskMakerBoss/Assets/Audio/SC_MaskMakerBoss_Teleport_Whoosh
root=Modulator_0
nodes [
 {id=Delay_0, children [Random_2], properties {delayMax=0.35, delayMin=0.35}}
 {id=Mixer_0, children [Random_6,Delay_0], properties {inputVolume [1,0.3]}}
 {id=Modulator_0, children [Mixer_0]}
 {
  id=Random_2
  children [WavePlayer_8, WavePlayer_9, WavePlayer_10, WavePlayer_11, WavePlayer_12, WavePlayer_13, WavePlayer_14]
  properties {weights [1,1,1,1,1,1,1]}
 }
 {
  id=Random_6
  children [WavePlayer_32, WavePlayer_33, WavePlayer_34, WavePlayer_35, WavePlayer_36, WavePlayer_37]
  properties {weights [1,1,1,1,1,1]}
 }
 {id=WavePlayer_10, properties {soundWaveAssetPtr=&/GiftedTraveler_Montage_Sparks_B_03}}
 {id=WavePlayer_11, properties {soundWaveAssetPtr=&/GiftedTraveler_Montage_Sparks_B_04}}
 {id=WavePlayer_12, properties {soundWaveAssetPtr=&/GiftedTraveler_Montage_Sparks_B_05}}
 {id=WavePlayer_13, properties {soundWaveAssetPtr=&/GiftedTraveler_Montage_Sparks_B_06}}
 {id=WavePlayer_14, properties {soundWaveAssetPtr=&/GiftedTraveler_Montage_Sparks_B_07}}
 {id=WavePlayer_32, properties {soundWaveAssetPtr=&/MaskMakerBoss_SynthWhooshB_06}}
 {id=WavePlayer_33, properties {soundWaveAssetPtr=&/MaskMakerBoss_SynthWhooshB_05}}
 {id=WavePlayer_34, properties {soundWaveAssetPtr=&/MaskMakerBoss_SynthWhooshB_04}}
 {id=WavePlayer_35, properties {soundWaveAssetPtr=&/MaskMakerBoss_SynthWhooshB_03}}
 {id=WavePlayer_36, properties {soundWaveAssetPtr=&/MaskMakerBoss_SynthWhooshB_02}}
 {id=WavePlayer_37, properties {soundWaveAssetPtr=&/MaskMakerBoss_SynthWhooshB_01}}
 {id=WavePlayer_8, properties {soundWaveAssetPtr=&/GiftedTraveler_Montage_Sparks_B_01}}
 {id=WavePlayer_9, properties {soundWaveAssetPtr=&/GiftedTraveler_Montage_Sparks_B_02}}
]
properties {
 attenuationSettings=&/ATT_MaskMakerBoss_Skills
 bDebug=true
 bOverrideConcurrency=true
 maxDistance=13500
 soundClassObject=&Fort_BossFight_SFX
 volumeMultiplier=0.7
 concurrencyOverrides {maxCount=3, resolutionRule=StopOldest}
}
