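# Sound cue for the Mask Maker boss's teleport, the WhooshB variant, routed to the boss-fight SFX class with boss-skill attenuation.
# A mixer layers randomly picked synth whoosh and teleport charge sounds with a delayed, slightly louder random teleport whoosh.
# A root modulator varies the mixed result on each play.
# A concurrency override caps simultaneous instances, stopping the oldest, and the debug flag is left on.
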SoundCue /MaskMakerBoss/Assets/Audio/SC_MaskMakerBoss_Teleport_WhooshB
root=Modulator_0
nodes [
 {id=Delay_1, children [Random_1], properties {delayMax=0.35, delayMin=0.35}}
 {id=Mixer_0, children [Random_6,Random_0,Delay_1], properties {inputVolume [1,1,1.2]}}
 {id=Modulator_0, children [Mixer_0], properties {pitchMin=1}}
 {id=Random_0, children [WavePlayer_0,WavePlayer_1,WavePlayer_2,WavePlayer_3], properties {weights [1,1,1,1]}}
 {id=Random_1, children [WavePlayer_4,WavePlayer_5,WavePlayer_6,WavePlayer_7], properties {weights [1,1,1,1]}}
 {
  id=Random_6
  children [WavePlayer_32, WavePlayer_33, WavePlayer_34, WavePlayer_35, WavePlayer_36, WavePlayer_37]
  properties {weights [1,1,1,1,1,1]}
 }
 {id=WavePlayer_0, properties {soundWaveAssetPtr=&/MaskMakerBoss_TeleportCharge_01}}
 {id=WavePlayer_1, properties {soundWaveAssetPtr=&/MaskMakerBoss_TeleportCharge_02}}
 {id=WavePlayer_2, properties {soundWaveAssetPtr=&/MaskMakerBoss_TeleportCharge_03}}
 {id=WavePlayer_3, properties {soundWaveAssetPtr=&/MaskMakerBoss_TeleportCharge_04}}
 {id=WavePlayer_32, properties {soundWaveAssetPtr=&/MaskMakerBoss_SynthWhooshB_06}}
 {id=WavePlayer_33, properties {soundWaveAssetPtr=&/MaskMakerBoss_SynthWhooshB_05}}
 {id=WavePlayer_34, properties {soundWaveAssetPtr=&/MaskMakerBoss_SynthWhooshB_04}}
 {id=WavePlayer_35, properties {soundWaveAssetPtr=&/MaskMakerBoss_SynthWhooshB_03}}
 {id=WavePlayer_36, properties {soundWaveAssetPtr=&/MaskMakerBoss_SynthWhooshB_02}}
 {id=WavePlayer_37, properties {soundWaveAssetPtr=&/MaskMakerBoss_SynthWhooshB_01}}
 {id=WavePlayer_4, properties {soundWaveAssetPtr=&/MaskMakerBoss_TeleportWhoosh_01}}
 {id=WavePlayer_5, properties {soundWaveAssetPtr=&/MaskMakerBoss_TeleportWhoosh_02}}
 {id=WavePlayer_6, properties {soundWaveAssetPtr=&/MaskMakerBoss_TeleportWhoosh_03}}
 {id=WavePlayer_7, properties {soundWaveAssetPtr=&/MaskMakerBoss_TeleportWhoosh_04}}
]
properties {
 attenuationSettings=&/ATT_MaskMakerBoss_Skills
 bDebug=true
 bOverrideConcurrency=true
 maxDistance=13500
 soundClassObject=&Fort_BossFight_SFX
 volumeMultiplier=0.85
 concurrencyOverrides {maxCount=3, resolutionRule=StopOldest}
}
